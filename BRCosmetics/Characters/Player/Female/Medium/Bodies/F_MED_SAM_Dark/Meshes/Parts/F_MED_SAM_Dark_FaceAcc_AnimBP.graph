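# Post-process-style hair AnimBP for the F_MED_SAM_Dark face accessory mesh, driving its dynamic hair bones.
# It blends attached-parent and owner-mesh arm poses, then aims side hair bones at the upper spine when rigid body is disabled.
# Head-space RigidBody and back-hair AnimDynamics follow, feeding a cached pose.
# In the front end, left and right hair master bones get additive yaw; the dyn_hair_bk_01 dynamics node is unconnected.
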AnimBlueprint /BRCosmetics/Characters/Player/Female/Medium/Bodies/F_MED_SAM_Dark/Meshes/Parts/F_MED_SAM_Dark_FaceAcc_AnimBP
target_skeleton=&/F_MED_SAM_Dark_FaceAcc_Skeleton
graphs [
 AnimGraph {nodes [
   Anim Dynamics - Bone: dyn_hair_bk_01!(ComponentPose, Alpha=1, GravityScale:float=1, GravityOverride:Vector="0,0,-980", ExternalForce="0,0,0") -> (Pose) {
    angularDampingOverride=1
    angularSpringConstant=100
    bAngularSpring=true
    bGravityOverrideInSimSpace=true
    bOverrideAngularDamping=true
    simulationSpace=RootRelative
    alphaBoolBlend {blendOption=Linear}
    boundBone {boneName=dyn_hair_bk_01}
    componentAppliedLinearAccClamp=<5000 5000 4000>
    componentLinearAccScale=<1.2 1.2 1.3>
    componentLinearVelScale=<1.2 1.2 1.3>
    gravityOverride {z=-980}
    physicsBodyDefinitions [{
      sphereCollisionRadius=0
      boundBone {boneName=dyn_hair_bk_01}
      boxExtents=<12 11 11>
      constraintSetup {angularLimitsMax=<45 5 35>, angularLimitsMin=<-45 -45 -35>, angularTarget=<0 0 0>}
      localJointOffset=<2 0 0>
     }]
    relativeSpaceBone {boneName=root}
   }
   n1 = "Transform (Modify) Bone - Bone: dyn_hair_l_MSTR"!(ComponentPose=Local To Component(LocalPose=Use cached pose 'Pose'()), Alpha=1, Rotation="0.000,5,-0.000") -> (Pose) {
    rotationMode=BMM_Additive
    rotationSpace=BCS_BoneSpace
    alphaBoolBlend {blendOption=Linear}
    boneToModify {boneName=dyn_hair_l_MSTR}
    rotation {yaw=5}
   }
   n2 = "Transform (Modify) Bone - Bone: dyn_hair_r_MSTR"!(ComponentPose=@n1, Alpha=1, Rotation="0.000,-5,-0.000") -> (Pose) {
    rotationMode=BMM_Additive
    rotationSpace=BCS_BoneSpace
    alphaBoolBlend {blendOption=Linear}
    boneToModify {boneName=dyn_hair_r_MSTR}
    rotation {yaw=-5}
   }
   Output Pose(Result=Blend Poses by bool(BlendPose_0=Component To Local(ComponentPose=@n2), BlendPose_1=Use cached pose 'Pose'(), bActiveValue=Get bIsInFrontEnd()) {blendType=Linear}) {layerGroup=None}
   n3 = Copy Pose From Mesh(SourceMeshComponent=Get Body_Mesh())
   n4 = Copy Pose From Mesh() {bUseAttachedParent=true}
   n5 = NOT Boolean(A=Get bRigidBodyEnabled())
   n6 = Look At - Bone: dyn_hair_r_01!(
    ComponentPose=Local To Component(LocalPose=Layered blend per bone(BasePose=@n4, BlendPoses_0=@n3) {layerSetup [{branchFilters [{boneName=upperarm_l}, {boneName=upperarm_r}]}]})
    bAlphaBoolEnabled:bool=@n5
   ) -> (Pose) {
    alphaInputType=Bool
    alphaBoolBlend {blendOption=Linear}
    boneToModify {boneName=dyn_hair_r_01}
    lookAtLocation=<-20 9.1329 3.7017>
    lookAtTarget {boneReference {boneName=spine_05}}
    lookAt_Axis {axis=<1 0>}
   }
   n7 = Look At - Bone: dyn_hair_r_03!(ComponentPose=@n6, bAlphaBoolEnabled:bool=@n5) -> (Pose) {
    alphaInputType=Bool
    alphaBoolBlend {blendOption=Linear}
    boneToModify {boneName=dyn_hair_r_03}
    lookAtLocation=<-18.2718 18.7955 8.6308>
    lookAtTarget {boneReference {boneName=spine_05}}
    lookAt_Axis {axis=<1 0>}
   }
   n8 = Look At - Bone: dyn_hair_r_04!(ComponentPose=@n7, bAlphaBoolEnabled:bool=@n5) -> (Pose) {
    alphaInputType=Bool
    alphaBoolBlend {blendOption=Linear}
    boneToModify {boneName=dyn_hair_r_04}
    lookAtLocation=<-18.2718 19.1741 10.6702>
    lookAtTarget {boneReference {boneName=spine_05}}
    lookAt_Axis {axis=<1 0>}
   }
   n9 = Look At - Bone: dyn_hair_r_06!(ComponentPose=@n8, bAlphaBoolEnabled:bool=@n5) -> (Pose) {
    alphaInputType=Bool
    alphaBoolBlend {blendOption=Linear}
    boneToModify {boneName=dyn_hair_r_06}
    lookAtLocation=<-18.2718 18.0165 10.5287>
    lookAtTarget {boneReference {boneName=spine_05}}
    lookAt_Axis {axis=<1 0>}
   }
   n10 = Look At - Bone: dyn_hair_l_02!(ComponentPose=@n9, bAlphaBoolEnabled:bool=@n5) -> (Pose) {
    alphaInputType=Bool
    alphaBoolBlend {blendOption=Linear}
    boneToModify {boneName=dyn_hair_l_02}
    lookAtLocation=<-20 4.0306 -7.1915>
    lookAtTarget {boneReference {boneName=spine_05}}
    lookAt_Axis {axis=<1 0>}
   }
   n11 = Look At - Bone: dyn_hair_l_04!(ComponentPose=@n10, bAlphaBoolEnabled:bool=@n5) -> (Pose) {
    alphaInputType=Bool
    alphaBoolBlend {blendOption=Linear}
    boneToModify {boneName=dyn_hair_l_04}
    lookAtLocation=<-20 20.4856 -7.2617>
    lookAtTarget {boneReference {boneName=spine_05}}
    lookAt_Axis {axis=<1 0>}
   }
   n12 = Look At - Bone: dyn_hair_l_06!(ComponentPose=@n11, bAlphaBoolEnabled:bool=@n5) -> (Pose) {
    alphaInputType=Bool
    alphaBoolBlend {blendOption=Linear}
    boneToModify {boneName=dyn_hair_l_06}
    lookAtLocation=<-20 14.7253 -10.5007>
    lookAtTarget {boneReference {boneName=spine_04}}
    lookAt_Axis {axis=<1 0>}
   }
   n13 = RigidBody(ComponentPose=@n12) {
    bClampLinearTranslationLimitToRefPose=true
    bOverrideWorldGravity=true
    overridePhysicsAsset=&/F_MED_SAM_dark_dyn_hair
    simulationSpace=BaseBoneSpace
    alphaBoolBlend {blendOption=Linear}
    baseBoneRef {boneName=head}
    componentAppliedLinearAccClamp=<500 500 500>
    componentLinearAccScale=<0.25 0.3 0.3>
    componentLinearVelScale=<0.2 0.2 0.2>
    overrideWorldGravity {z=-980}
   }
   n14 = Anim Dynamics - Bone: dyn_hair_bk_02!(ComponentPose=@n13, Alpha=1, GravityScale:float=1, GravityOverride:Vector="0,0,-980", ExternalForce="0,0,0") -> (Pose) {
    angularDampingOverride=1
    angularSpringConstant=100
    bAngularSpring=true
    bGravityOverrideInSimSpace=true
    bOverrideAngularDamping=true
    simulationSpace=RootRelative
    alphaBoolBlend {blendOption=Linear}
    boundBone {boneName=dyn_hair_bk_02}
    componentAppliedLinearAccClamp=<5000 5000 4000>
    componentLinearAccScale=<1.2 1.2 1.4>
    componentLinearVelScale=<1.2 1.2 1.4>
    gravityOverride {z=-980}
    physicsBodyDefinitions [{
      sphereCollisionRadius=0
      boundBone {boneName=dyn_hair_bk_02}
      boxExtents=<12 11 11>
      constraintSetup {angularLimitsMax=<45 5 35>, angularLimitsMin=<-45 -45 -35>, angularTarget=<0 0 0>}
      localJointOffset=<2 0 0>
     }]
    relativeSpaceBone {boneName=root}
   }
   Save cached pose 'Pose'(Pose=Component To Local(ComponentPose=@n14)) {cacheName=Pose}
  ]}
 EventGraph {nodes [
   Event BlueprintInitializeAnimation()
   > Set Body_Mesh(Body_Mesh=GetOwnerSkeletalMeshComponent())
   Event BlueprintUpdateAnimation()
  ]}
]
variables [Body_Mesh {type=SkeletalMeshComponent}]
